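# Post-sim anim blueprint for the MollyKit tail cosmetic, adding trailing motion to the tail and managing its visibility.
# AnimGraph runs the input pose through a ten-bone Trail controller on C_Tail_A_9_Jnt, with alpha and relaxation scale read from variables.
# On customization load, it checks for tail and backbling, hiding the tail under long backblings (capes).
# Otherwise the tail shows and Cape_Tail combo flags are set, split by front end versus in game.
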AnimBlueprint /BRCosmetics/Accessories/FORT_Tails/Tail_F_MED_MollyKit/Meshes/F_MED_MollyKit_Tail_AnimBP_PostSim
target_skeleton=/Game/Accessories/FORT_Tails/Common/Fortnite_Base_Tail_Skeleton
graphs [
 AnimGraph {nodes [
   n1 = Trail controller - Bone: C_Tail_A_9_Jnt!(
    ComponentPose=Local To Component(LocalPose=Input Pose())
    Alpha=Get TrailAlpha() -> (TrailAlpha:double)
    RelaxationSpeedScale:float=Get RelaxSpeed() -> (RelaxSpeed:double)
   ) -> (Pose) {
    bLimitRotation=true
    chainLength=10
    maxDeltaTime=0.0333
    alphaScaleBiasClamp {bInterpResult=true}
    rotationLimits [
     {limitMax=<180 180 180>}
     {limitMax=<180 180 180>}
     {limitMax=<180 180 180>}
     {limitMax=<180 180 180>}
     {limitMax=<180 180 180>}
     {limitMax=<180 180 180>}
     {limitMax=<180 180 180>}
     {limitMax=<180 180 180>}
     {limitMax=<180 180 180>}
     {limitMax=<180 180 180>}
    ]
    rotationOffsets [<0 0 0>, <0 0 0>, <0 0 0>, <0 0 0>, <0 0 0>, <0 0 0>, <0 0 0>, <0 0 0>, <0 0 0>, <0 0 0>]
    trailBone {boneName=C_Tail_A_9_Jnt}
    trailRelaxationSpeed {editorCurveData {keys [<at 0 10 RCIM_Linear>,<at 0.5107 8.1304 RCIM_Linear>,<at 1 5 RCIM_Linear>]}}
   }
   Output Pose(Result=Component To Local(ComponentPose=@n1)) {layerGroup=None}
  ]}
 EventGraph {nodes [
   n1 = Event OnCharacterCustomizationLoaded()
   > Set Character Mesh(Character Mesh:Skeletal Mesh Component Object Reference=@n1.CustomBaseMesh) -> (Output_Get:Skeletal Mesh Component Object Reference)
   > Set TailMesh(TailMesh:Skeletal Mesh Component Object Reference=Get Array Item with EFort Custom Part Type(Array=@n1.CustomMeshParts, PartType=MiscOrTail)) -> (Output_Get:Skeletal Mesh Component Object Reference)
   n2 = CheckForBackbling()
   > Set isWearingBackbling(isWearingBackbling=@n2)
   > Set isWearingCape(isWearingCape=@n2.bIsBackblingLong)
   > Set Character_has_tail(Character_has_tail:bool=CheckForTail!(self:Custom Character Part Anim Instance Object Reference) -> (bHasTail:bool)) -> (Output_Get:bool)
   > Branch(Condition=Get Character_has_tail() -> (Character_has_tail:bool))
   > n3 = Branch(Condition=Get isWearingCape() -> (isWearingCape:bool))
   n4 = Get TailMesh() -> (TailMesh:Skeletal Mesh Component Object Reference)
   > SetVisibility!(execute, self:Scene Component Object Reference=@n4, bNewVisibility:bool=false, bPropagateToChildren:bool=false) -> (then)
   > Set Cape_Tail Combo(Cape_Tail Combo:bool=false) -> (Output_Get:bool)
   > Set Cape_Tail(Cape_Tail:bool=false) -> (Output_Get:bool)
   SetVisibility(execute=@n3.else, self=@n4, bNewVisibility=true, bPropagateToChildren=false)
   > Set Cape_Tail(Cape_Tail=true)
   > n5 = Branch(Condition=Get bIsInFrontEnd())
   > Set Cape_Tail Combo_Frontend(Cape_Tail Combo_Frontend:bool=true) -> (Output_Get:bool)
   Set Cape_Tail Combo(execute=@n5.else, Cape_Tail Combo=true)
   GetSkeletalMeshForPartType(self=Get Target() -> (Target:Fort Player Pawn Object Reference), PartType=MiscOrTail)
  ]}
]
variables [
 As Player Pawn Athena {type=PlayerPawn_Athena_C}
 Body Mesh {type=SkeletalMeshComponentBudgeted}
 Cape_Tail {type=boolean}
 Cape_Tail Combo {type=boolean}
 Cape_Tail Combo_Frontend {type=boolean}
 Character Mesh {type=SkeletalMeshComponent}
 Character_has_tail {type=boolean}
 isWearingBackbling {type=boolean}
 isWearingCape {type=boolean}
 Pawn Owner {type=PlayerPawn_Athena_C}
 PawnOwner {type=Pawn, default=null}
 RelaxSpeed {type=number}
 TailMesh {type=SkeletalMeshComponent}
 Target {type=FortPlayerPawn}
 TrailAlpha {type=number}
]
